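# Niagara system for the GetTheHorns emote that kicks up dust and rock debris from the player's legs.
# Initial-left, left and right leg emitter pairs use staggered loop timings and spawn at offsets from the User.SkeletalMesh.
# CPU dust sprites combine bursts with a decaying spawn rate, pushed along -Y under drag and curve-driven dynamic material parameters.
# GPU rock sprites burst, fall under gravity, bounce with collision and shrink out.
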NiagaraSystem /BRCosmetics/Animation/Game/MainPlayer/Emotes/GetTheHorns/FX/NS_GetTheHorns
dependencies {
 usedDataInterfaces [&NiagaraDataInterfaceCurve, &NiagaraDataInterfaceColorCurve]
 usedDynamicInputs []
 usedModules [&SystemState]
}
emitters {
 Initial_Left_Leg_Dust_Kickoff {
  emitterUpdateScript [
   &EmitterState(Life Cycle Mode=::NewEnumerator1, Loop Behavior=::NewEnumerator1, Loop Delay=5.6, Loop Duration=1.15, UseLoopDelay=true)
   &SpawnBurst_Instantaneous(Spawn Count=10, Use Spawn Probability=false)
   &SpawnRate(
    SpawnGroup=1
    SpawnRate=&FloatFromCurve(
     FloatCurve=DataInterface() {propertyValues {Curve {keys [<at 0 1 RCIM_Cubic>,<at 0.5 0 RCIM_Cubic>]}}}
     CurveIndex=$Emitter.NormalizedLoopAge
     Scale Curve=30
    )
    Use Spawn Probability=false
   )
  ]
  particleSpawnScript [
   &InitializeParticle(
    Mass Mode=::NewEnumerator1
    Sprite Rotation Angle=&RandomRangeFloat(Minimum=0, Maximum=360)
    Sprite Rotation Mode=::NewEnumerator0
    Sprite Size Mode=::NewEnumerator1
    Uniform Sprite Size Max=60
    Uniform Sprite Size Min=50
    UsePositionOffset=false
    Write Material Random=true
   )
   &SkeletalMeshLocation(
    Bone Sampling Mode=::NewEnumerator4
    Inherit Velocity=false
    MaskSpawnGroup=false
    Sampled Position Offset=&MakeVector(
     X=15
     Y=&MakeCustomFloatFromBool(
      Bool=HlslExpression() {hlslExpression="Emitter.SpawnGroup == 1"}
      True Float=&RandomRangeFloat(Minimum=-36, Maximum=-50)
      False Float=-30
     )
     Z=10
    )
    Sampled Position Offset Coordinate Space:ENiagaraMeshTransforms=::NewEnumerator3
    Skeletal Mesh=$User.SkeletalMesh
   )
   &SphereLocation(
    MaskSpawnGroup=false
    Override Seed=false
    Sphere Radius=6
    Surface Only=false
    UseNonUniformScale=true
    UseOffset=false
    UseSphereOrientation=false
   )
  ]
  particleUpdateScript [
   &ParticleState()
   &Color(
    Color=&ColorFromCurve(ColorCurve=DataInterface() {propertyValues {
       AlphaCurve {keys [<at 0 1 RCIM_Cubic>,<at 1 1 RCIM_Cubic>]}
       BlueCurve {keys [<at 0 0.4799 RCIM_Cubic>,<at 1 0.1934 RCIM_Cubic>]}
       GreenCurve {keys [<at 0 0.7637 RCIM_Cubic>,<at 1 0.2126 RCIM_Cubic>]}
       RedCurve {keys [<at 0 1 RCIM_Cubic>,<at 1 0.2292 RCIM_Cubic>]}
      }})
    Scale Alpha=&FloatFromCurve(
     FloatCurve=DataInterface() {propertyValues {Curve {keys [
         {interpMode=RCIM_Cubic, leaveTangent=11.125, tangentMode=RCTM_Break, time=0, value=0}
         {interpMode=RCIM_Cubic, leaveTangent=-2.2426, tangentMode=RCTM_Break, time=0.25, value=1.2}
         {interpMode=RCIM_Cubic, tangentMode=RCTM_Break, time=1, value=0}
        ]}}}
     CurveIndex=&MakeCustomFloatFromBool(Bool=HlslExpression() {hlslExpression="Emitter.SpawnGroup == 1"}, True Float=$Emitter.LoopedAge, False Float=$Particles.NormalizedAge)
     Scale Curve=0.2
    )
   )
   &Drag(
    Drag=&FloatFromCurve(FloatCurve=DataInterface() {propertyValues {Curve {keys [{interpMode=RCIM_Cubic, tangentMode=RCTM_Break, time=0, value=1}, {interpMode=RCIM_Cubic, tangentMode=RCTM_Break, time=0.3, value=500}]}}})
    Use Linear Drag=true
    Use Rotational Drag=false
   )
   &GravityForce(Gravity=<0 0 -2500>)
   &DynamicMaterialParameters(
    Index 0 Param 1=&Add_Float(
     A=$Particles.RandomFloat
     B=&FloatFromCurve(
      FloatCurve=DataInterface() {propertyValues {Curve {keys [{arriveTangent=2.7533, time=0, value=1}, {arriveTangent=0.0529, leaveTangent=0.2021, time=1, value=1.5}]}}}
      CurveIndex=&Add_Float(A=&Multiply_Float(A=$Particles.Initial.MaterialRandom, B=0.5), B=$Particles.NormalizedAge)
     )
    )
    Index 0 Param 2=&FloatFromCurve(FloatCurve=DataInterface() {propertyValues {Curve {keys [
         {arriveTangent=0.0453, interpMode=RCIM_Cubic, tangentMode=RCTM_Break, time=0, value=-1}
         {arriveTangent=0.0604, interpMode=RCIM_Cubic, leaveTangent=0.0604, tangentMode=RCTM_User, time=1, value=-1}
        ]}}})
    Index 0 Param 3=&FloatFromCurve(FloatCurve=DataInterface() {propertyValues {Curve {keys [
         {arriveTangent=-4.113, interpMode=RCIM_Cubic, leaveTangent=-2.1889, tangentMode=RCTM_User, time=0, value=1}
         {arriveTangent=0.2347, interpMode=RCIM_Cubic, leaveTangent=0.2723, tangentMode=RCTM_Break, time=1, value=0}
        ]}}})
    Index 0 Param 4=&FloatFromCurve(FloatCurve=DataInterface() {propertyValues {Curve {keys [<at 0 1>,<at 1 0>]}}})
    Index 1 Param 1=&FloatFromCurve(
     FloatCurve=DataInterface() {propertyValues {Curve {keys [
         {interpMode=RCIM_Cubic, leaveTangent=4.7088, tangentMode=RCTM_Break, time=0, value=0}
         {arriveTangent=4.2257, interpMode=RCIM_Cubic, leaveTangent=4.2257, tangentMode=RCTM_User, time=0.25, value=1}
        ]}}}
     Scale Curve=&RandomRangeFloat(Minimum=1.5, Maximum=1.5)
    )
    Index 1 Param 2=&FloatFromCurve(
     FloatCurve=DataInterface() {propertyValues {Curve {keys [<at 0 0>, {leaveTangent=0.9066, time=1, value=1}]}}}
     Scale Curve=&RandomRangeFloat(Minimum=0.5, Maximum=0.6)
    )
    Index 1 Param 3=0.25
    Index 1 Param 4=&FloatFromCurve(
     FloatCurve=DataInterface() {propertyValues {Curve {keys [{arriveTangent=-0.3224, interpMode=RCIM_Cubic, leaveTangent=-0.3224, tangentMode=RCTM_User, time=0, value=1}, <at 1 0.9>]}}}
     Scale Curve=&RandomRangeFloat(Minimum=1.1, Maximum=1.4)
    )
    Index 2 Param 1=&FloatFromCurve(FloatCurve=DataInterface() {propertyValues {Curve {keys [{interpMode=RCIM_Cubic, leaveTangent=1.4673, tangentMode=RCTM_Break, time=0, value=0}, <at 1 1>]}}})
    Index 2 Param 2=&FloatFromCurve(
     FloatCurve=DataInterface() {propertyValues {Curve {keys [
         {arriveTangent=1.9056, interpMode=RCIM_Cubic, leaveTangent=1.9056, tangentMode=RCTM_User, time=0, value=0.2}
         {arriveTangent=0.0259, interpMode=RCIM_Cubic, leaveTangent=0.0259, tangentMode=RCTM_User, time=1, value=1.4}
        ]}}}
     Scale Curve=&RandomRangeFloat(Minimum=0.2, Maximum=0.7)
    )
    Index 2 Param 3=&FloatFromCurve(FloatCurve=DataInterface() {propertyValues {Curve {keys [<at 0 1>, {leaveTangent=11.1233, time=1, value=2}]}}})
    Module.Param0WriteEnabled=true
    Module.Param0WriteEnabled1=false
    Module.Param0WriteEnabled2=true
    Module.Param1WriteEnabled=false
    Module.Param1WriteEnabled1=false
    Module.Param1WriteEnabled2=true
    Module.Param2WriteEnabled=true
    Module.Param2WriteEnabled1=true
    Module.Param2WriteEnabled2=true
    Module.Param3WriteEnabled=true
    Module.Param3WriteEnabled1=true
    Module.Param3WriteEnabled2=true
    Write Parameter Index 1=true
    Write Parameter Index 2=true
   )
   &ScaleSpriteSize(Scale Factor=&Vector2DFromFloat(Value=&FloatFromCurve(FloatCurve=DataInterface() {propertyValues {Curve {keys [
          {interpMode=RCIM_Cubic, leaveTangent=-0.0041, tangentMode=RCTM_Break, time=0, value=0.42}
          {interpMode=RCIM_Cubic, tangentMode=RCTM_Break, time=0.5, value=1}
         ]}}})))
   &AddVelocity(Velocity=&MakeVector(
     X=0
     Y=-250
     Z=&FloatFromCurve(FloatCurve=DataInterface() {propertyValues {Curve {keys [
          {interpMode=RCIM_Cubic, leaveTangent=-0.7633, tangentMode=RCTM_Break, time=0, value=1}
          {arriveTangent=-4.7682, interpMode=RCIM_Cubic, tangentMode=RCTM_Break, time=1, value=100}
         ]}}})
    ))
   &SolveForcesAndVelocity(Clamp Velocity=false, Limit Acceleration=false)
  ]
  rendererClasses [&NiagaraSpriteRendererProperties]
  renderers [{rendererClass=&NiagaraSpriteRendererProperties}]
 }
 Initial_Left_Leg_Rock_Kickoff {
  simTarget=GPUComputeSim
  emitterUpdateScript [
   &EmitterScalability(
    Module.bApplySpawnCountScale=false
    Module.bApplySpawnCountScaleByDistanceFraction=false
    Module.bUseMaxDistance=false
    Module.bUseMinDistance=false
   ) {enabled=false}
   &EmitterLifeCycle(NextLoopDuration=4.4208) {enabled=false}
   &EmitterState(Life Cycle Mode=::NewEnumerator1, Loop Behavior=::NewEnumerator1, Loop Delay=5.6, Loop Duration=1.15, UseLoopDelay=true)
   &SpawnBurst_Instantaneous(Spawn Count=&RandomRangeInt(Minimum=6, Maximum=10), Spawn Time=0.1, Use Spawn Probability=false)
  ]
  particleSpawnScript [
   /Niagara/Modules/Spawn/Initialization/InitializeParticle(
    Lifetime=&UniformRangedFloat(Minimum=0.5, Maximum=0.7)
    Mass=&RandomRangeFloat(Minimum=0.5, Maximum=2.3)
    Module.Write Color=true
    Module.Write Lifetime=true
    Module.Write Mass=true
    Module.Write Material Random=true
    Module.Write Position=false
    Module.Write Scale=false
    Module.Write SpriteRotation=false
    Module.Write SpriteSize=true
    Position=&SimulationPosition()
    Sprite Size=&Vector2DFromFloat(Value=&FloatFromCurve(
      FloatCurve=DataInterface() {propertyValues {Curve {keys [<at 0.1 1>,<at 0.2 0.5>,<at 1.0011 0.5>]}}}
      Scale Curve=&RandomRangeFloat(Minimum=1, Maximum=8)
     ))
   )
   &SkeletalMeshLocation(
    Bone Sampling Mode=::NewEnumerator4
    Inherit Velocity=false
    MaskSpawnGroup=false
    Sampled Position Offset=&MakeVector(
     X=15
     Y=&MakeCustomFloatFromBool(
      Bool=HlslExpression() {hlslExpression="Emitter.SpawnGroup == 1"}
      True Float=&RandomRangeFloat(Minimum=-36, Maximum=-50)
      False Float=-30
     )
     Z=10
    )
    Sampled Position Offset Coordinate Space:ENiagaraMeshTransforms=::NewEnumerator3
    Skeletal Mesh=$User.SkeletalMesh
   )
   &SphereLocation(
    MaskSpawnGroup=false
    Override Seed=false
    Sphere Radius=10
    Surface Only=false
    UseNonUniformScale=false
    UseOffset=false
    UseSphereOrientation=false
   )
   &AddVelocity(Velocity=&MakeVector(
     X=0
     Y=&RandomRangeFloat(Minimum=-10, Maximum=-300)
     Z=&RandomRangeFloat(Minimum=40, Maximum=200)
    ))
  ]
  particleUpdateScript [
   &ParticleState()
   &Drag(Drag=&RandomRangeFloat(Minimum=3, Maximum=5), Use Linear Drag=true, Use Rotational Drag=false)
   &GravityForce(Gravity=<0 0 -900>)
   &Collision(
    Advanced Aging Rate=1
    Age Colliding Particles=false
    Friction=0.8
    Friction During a Bounce=0.1
    Kill Offscreen Particles=true
    Method for Calculating Particle Radius:ENiagaraCalculateRadiusOptions=::NewEnumerator1
    Module.Control Rest State Via Surface Normal=false
    Module.EnableMaxCollisionCount=false
    Module.Randomize Collision Normal Vector=true
    Particle Radius Scale=0.1
    Randomize Collision Normal=0.5
    Restitution=&UniformRangedFloat(Minimum=0.2, Maximum=0.6)
   )
   &SolveForcesAndVelocity(Clamp Velocity=false, Limit Acceleration=false)
   &ScaleColor(
    Scale Alpha=&FloatFromCurve(FloatCurve=DataInterface() {propertyValues {Curve {keys [
         {arriveTangent=3.5481, interpMode=RCIM_Cubic, leaveTangent=3.5481, tangentMode=RCTM_User, time=0, value=1}
         {arriveTangent=-1, interpMode=RCIM_Cubic, leaveTangent=-1, time=0.2243, value=1}
         {arriveTangent=-0.9169, interpMode=RCIM_Cubic, leaveTangent=-0.9169, tangentMode=RCTM_User, time=1, value=0}
        ]}}})
    Scale RGB=<1.5 1.5 1.5>
    ScaleA=true
    ScaleRGB=true
   ) {enabled=false}
   &ScaleSpriteSize(Scale Factor=&Vector2DFromFloat(Value=&FloatFromCurve(FloatCurve=DataInterface() {propertyValues {Curve {keys [<at 0.7 0.9701>,<at 0.9011 0.9191>,<at 1.0011 0.002>]}}})))
   &ScaleSpriteSizeBySpeed(Max Scale Factor=<1 1.5>, Min Scale Factor=<0.7 0.7>, Velocity Threshold=250) {enabled=false}
   &Color(
    Color=<rgba 0.2917 0.2917 0.2917 1>
    Scale Alpha=&FloatFromCurve(FloatCurve=DataInterface() {propertyValues {Curve {keys [
         {arriveTangent=3.5481, interpMode=RCIM_Cubic, leaveTangent=3.5481, tangentMode=RCTM_User, time=0, value=1}
         {arriveTangent=-1, interpMode=RCIM_Cubic, leaveTangent=-1, time=0.2243, value=1}
         {arriveTangent=-0.9169, interpMode=RCIM_Cubic, leaveTangent=-0.9169, tangentMode=RCTM_User, time=1, value=0}
        ]}}})
   )
  ]
  rendererClasses [&NiagaraSpriteRendererProperties]
  renderers [{rendererClass=&NiagaraSpriteRendererProperties}]
 }
 Left_Leg_Dust_Kickoff {
  emitterUpdateScript [
   &EmitterState(
    Delay First Loop Only=true
    Life Cycle Mode=::NewEnumerator1
    Loop Behavior=::NewEnumerator2
    Loop Count=2
    Loop Delay=0.6
    Loop Duration=1.15
    Recalculate Duration Each Loop=true
    UseLoopDelay=true
   )
   &SpawnBurst_Instantaneous(Spawn Count=10, Use Spawn Probability=false)
   &SpawnRate(
    SpawnGroup=1
    SpawnRate=&FloatFromCurve(
     FloatCurve=DataInterface() {propertyValues {Curve {keys [<at 0 1 RCIM_Cubic>,<at 0.5 0 RCIM_Cubic>]}}}
     CurveIndex=$Emitter.NormalizedLoopAge
     Scale Curve=15
    )
    Use Spawn Probability=false
   )
  ]
  particleSpawnScript [
   /Niagara/Modules/Spawn/Initialization/V2/InitializeParticle(
    Mass Mode=::NewEnumerator1
    Sprite Rotation Angle=&RandomRangeFloat(Minimum=0, Maximum=360)
    Sprite Rotation Mode=::NewEnumerator0
    Sprite Size Mode=::NewEnumerator3
    Uniform Sprite Size=50
    UsePositionOffset=false
    Write Material Random=true
   )
   &SkeletalMeshLocation(
    Bone Sampling Mode=::NewEnumerator4
    Inherit Velocity=false
    MaskSpawnGroup=false
    Sampled Position Offset=&MakeVector(
     X=15
     Y=&MakeCustomFloatFromBool(
      Bool=HlslExpression() {hlslExpression="Emitter.SpawnGroup == 1"}
      True Float=&RandomRangeFloat(Minimum=-36, Maximum=-50)
      False Float=-30
     )
     Z=10
    )
    Sampled Position Offset Coordinate Space:ENiagaraMeshTransforms=::NewEnumerator3
    Skeletal Mesh=$User.SkeletalMesh
   )
   &SphereLocation(
    MaskSpawnGroup=false
    Override Seed=false
    Sphere Radius=6
    Surface Only=false
    UseNonUniformScale=true
    UseOffset=false
    UseSphereOrientation=false
   )
  ]
  particleUpdateScript [
   &ParticleState()
   &Color(
    Color=&ColorFromCurve(ColorCurve=DataInterface() {propertyValues {
       AlphaCurve {keys [<at 0 1 RCIM_Cubic>,<at 1 1 RCIM_Cubic>]}
       BlueCurve {keys [<at 0 0.4799 RCIM_Cubic>,<at 1 0.1934 RCIM_Cubic>]}
       GreenCurve {keys [<at 0 0.7637 RCIM_Cubic>,<at 1 0.2126 RCIM_Cubic>]}
       RedCurve {keys [<at 0 1 RCIM_Cubic>,<at 1 0.2292 RCIM_Cubic>]}
      }})
    Scale Alpha=&FloatFromCurve(
     FloatCurve=DataInterface() {propertyValues {Curve {keys [
         {interpMode=RCIM_Cubic, leaveTangent=11.125, tangentMode=RCTM_Break, time=0, value=0}
         {interpMode=RCIM_Cubic, leaveTangent=-2.2426, tangentMode=RCTM_Break, time=0.25, value=1.2}
         {interpMode=RCIM_Cubic, tangentMode=RCTM_Break, time=1, value=0}
        ]}}}
     CurveIndex=&MakeCustomFloatFromBool(Bool=HlslExpression() {hlslExpression="Emitter.SpawnGroup == 1"}, True Float=$Emitter.LoopedAge, False Float=$Particles.NormalizedAge)
     Scale Curve=0.2
    )
   )
   &Drag(
    Drag=&FloatFromCurve(FloatCurve=DataInterface() {propertyValues {Curve {keys [{interpMode=RCIM_Cubic, tangentMode=RCTM_Break, time=0, value=1}, {interpMode=RCIM_Cubic, tangentMode=RCTM_Break, time=0.3, value=500}]}}})
    Use Linear Drag=true
    Use Rotational Drag=false
   )
   &GravityForce(Gravity=<0 0 -2500>)
   &DynamicMaterialParameters(
    Index 0 Param 1=&Add_Float(
     A=$Particles.RandomFloat
     B=&FloatFromCurve(
      FloatCurve=DataInterface() {propertyValues {Curve {keys [{arriveTangent=2.7533, time=0, value=1}, {arriveTangent=0.0529, leaveTangent=0.2021, time=1, value=1.5}]}}}
      CurveIndex=&Add_Float(A=&Multiply_Float(A=$Particles.Initial.MaterialRandom, B=0.5), B=$Particles.NormalizedAge)
     )
    )
    Index 0 Param 2=&FloatFromCurve(FloatCurve=DataInterface() {propertyValues {Curve {keys [
         {arriveTangent=0.0453, interpMode=RCIM_Cubic, tangentMode=RCTM_Break, time=0, value=-1}
         {arriveTangent=0.0604, interpMode=RCIM_Cubic, leaveTangent=0.0604, tangentMode=RCTM_User, time=1, value=-1}
        ]}}})
    Index 0 Param 3=&FloatFromCurve(FloatCurve=DataInterface() {propertyValues {Curve {keys [
         {arriveTangent=-4.113, interpMode=RCIM_Cubic, leaveTangent=-2.1889, tangentMode=RCTM_User, time=0, value=1}
         {arriveTangent=0.2347, interpMode=RCIM_Cubic, leaveTangent=0.2723, tangentMode=RCTM_Break, time=1, value=0}
        ]}}})
    Index 0 Param 4=&FloatFromCurve(FloatCurve=DataInterface() {propertyValues {Curve {keys [<at 0 1>,<at 1 0>]}}})
    Index 1 Param 1=&FloatFromCurve(
     FloatCurve=DataInterface() {propertyValues {Curve {keys [
         {interpMode=RCIM_Cubic, leaveTangent=4.7088, tangentMode=RCTM_Break, time=0, value=0}
         {arriveTangent=4.2257, interpMode=RCIM_Cubic, leaveTangent=4.2257, tangentMode=RCTM_User, time=0.25, value=1}
        ]}}}
     Scale Curve=&RandomRangeFloat(Minimum=1.5, Maximum=1.5)
    )
    Index 1 Param 2=&FloatFromCurve(
     FloatCurve=DataInterface() {propertyValues {Curve {keys [<at 0 0>, {leaveTangent=0.9066, time=1, value=1}]}}}
     Scale Curve=&RandomRangeFloat(Minimum=0.5, Maximum=0.6)
    )
    Index 1 Param 3=0.25
    Index 1 Param 4=&FloatFromCurve(
     FloatCurve=DataInterface() {propertyValues {Curve {keys [{arriveTangent=-0.3224, interpMode=RCIM_Cubic, leaveTangent=-0.3224, tangentMode=RCTM_User, time=0, value=1}, <at 1 0.9>]}}}
     Scale Curve=&RandomRangeFloat(Minimum=1.1, Maximum=1.4)
    )
    Index 2 Param 1=&FloatFromCurve(FloatCurve=DataInterface() {propertyValues {Curve {keys [{interpMode=RCIM_Cubic, leaveTangent=1.4673, tangentMode=RCTM_Break, time=0, value=0}, <at 1 1>]}}})
    Index 2 Param 2=&FloatFromCurve(
     FloatCurve=DataInterface() {propertyValues {Curve {keys [
         {arriveTangent=1.9056, interpMode=RCIM_Cubic, leaveTangent=1.9056, tangentMode=RCTM_User, time=0, value=0.2}
         {arriveTangent=0.0259, interpMode=RCIM_Cubic, leaveTangent=0.0259, tangentMode=RCTM_User, time=1, value=1.4}
        ]}}}
     Scale Curve=&RandomRangeFloat(Minimum=0.2, Maximum=0.7)
    )
    Index 2 Param 3=&FloatFromCurve(FloatCurve=DataInterface() {propertyValues {Curve {keys [<at 0 1>, {leaveTangent=11.1233, time=1, value=2}]}}})
    Module.Param0WriteEnabled=true
    Module.Param0WriteEnabled1=false
    Module.Param0WriteEnabled2=true
    Module.Param1WriteEnabled=false
    Module.Param1WriteEnabled1=false
    Module.Param1WriteEnabled2=true
    Module.Param2WriteEnabled=true
    Module.Param2WriteEnabled1=true
    Module.Param2WriteEnabled2=true
    Module.Param3WriteEnabled=true
    Module.Param3WriteEnabled1=true
    Module.Param3WriteEnabled2=true
    Write Parameter Index 1=true
    Write Parameter Index 2=true
   )
   &ScaleSpriteSize(Scale Factor=&Vector2DFromFloat(Value=&FloatFromCurve(FloatCurve=DataInterface() {propertyValues {Curve {keys [
          {interpMode=RCIM_Cubic, leaveTangent=-0.0046, tangentMode=RCTM_Break, time=0, value=0.42}
          {interpMode=RCIM_Cubic, tangentMode=RCTM_Break, time=0.5, value=1}
         ]}}})))
   &AddVelocity(Velocity=&MakeVector(
     X=0
     Y=-250
     Z=&FloatFromCurve(FloatCurve=DataInterface() {propertyValues {Curve {keys [
          {interpMode=RCIM_Cubic, leaveTangent=-7.4955, tangentMode=RCTM_Break, time=0, value=1}
          {arriveTangent=-73.4151, interpMode=RCIM_Cubic, tangentMode=RCTM_Break, time=1, value=100}
         ]}}})
    ))
   &SolveForcesAndVelocity(Clamp Velocity=false, Limit Acceleration=false)
  ]
  rendererClasses [&NiagaraSpriteRendererProperties]
  renderers [{rendererClass=&NiagaraSpriteRendererProperties}]
 }
 Left_Leg_Rock_Kickoff {
  simTarget=GPUComputeSim
  emitterUpdateScript [
   &EmitterScalability(
    Module.bApplySpawnCountScale=false
    Module.bApplySpawnCountScaleByDistanceFraction=false
    Module.bUseMaxDistance=false
    Module.bUseMinDistance=false
   ) {enabled=false}
   &EmitterLifeCycle(NextLoopDuration=4.4208) {enabled=false}
   &EmitterState(
    Delay First Loop Only=true
    Life Cycle Mode=::NewEnumerator1
    Loop Behavior=::NewEnumerator2
    Loop Count=2
    Loop Delay=0.6
    Loop Duration=1.15
    Recalculate Duration Each Loop=true
    UseLoopDelay=true
   )
   &SpawnBurst_Instantaneous(Spawn Count=&RandomRangeInt(Minimum=6, Maximum=10), Spawn Time=0.1, Use Spawn Probability=false)
  ]
  particleSpawnScript [
   /Niagara/Modules/Spawn/Initialization/InitializeParticle(
    Lifetime=&UniformRangedFloat(Minimum=0.5, Maximum=0.7)
    Mass=&RandomRangeFloat(Minimum=0.5, Maximum=2.3)
    Module.Write Color=true
    Module.Write Lifetime=true
    Module.Write Mass=true
    Module.Write Material Random=true
    Module.Write Position=false
    Module.Write Scale=false
    Module.Write SpriteRotation=false
    Module.Write SpriteSize=true
    Position=&SimulationPosition()
    Sprite Size=&Vector2DFromFloat(Value=&FloatFromCurve(
      FloatCurve=DataInterface() {propertyValues {Curve {keys [<at 0.1 1>,<at 0.2 0.5>,<at 1.0011 0.5>]}}}
      Scale Curve=&RandomRangeFloat(Minimum=1, Maximum=6)
     ))
   )
   &SkeletalMeshLocation(
    Bone Sampling Mode=::NewEnumerator4
    Inherit Velocity=false
    MaskSpawnGroup=false
    Sampled Position Offset=&MakeVector(
     X=15
     Y=&MakeCustomFloatFromBool(
      Bool=HlslExpression() {hlslExpression="Emitter.SpawnGroup == 1"}
      True Float=&RandomRangeFloat(Minimum=-36, Maximum=-50)
      False Float=-30
     )
     Z=10
    )
    Sampled Position Offset Coordinate Space:ENiagaraMeshTransforms=::NewEnumerator3
    Skeletal Mesh=$User.SkeletalMesh
   )
   &SphereLocation(
    MaskSpawnGroup=false
    Override Seed=false
    Sphere Radius=10
    Surface Only=false
    UseNonUniformScale=false
    UseOffset=false
    UseSphereOrientation=false
   )
   &AddVelocity(Velocity=&MakeVector(
     X=0
     Y=&RandomRangeFloat(Minimum=-10, Maximum=-300)
     Z=&RandomRangeFloat(Minimum=40, Maximum=200)
    ))
  ]
  particleUpdateScript [
   &ParticleState()
   &Drag(Drag=&RandomRangeFloat(Minimum=3, Maximum=5), Use Linear Drag=true, Use Rotational Drag=false)
   &GravityForce(Gravity=<0 0 -900>)
   &Collision(
    Advanced Aging Rate=1
    Age Colliding Particles=false
    Friction=0.8
    Friction During a Bounce=0.1
    Kill Offscreen Particles=true
    Method for Calculating Particle Radius:ENiagaraCalculateRadiusOptions=::NewEnumerator1
    Module.Control Rest State Via Surface Normal=false
    Module.EnableMaxCollisionCount=false
    Module.Randomize Collision Normal Vector=true
    Particle Radius Scale=0.1
    Randomize Collision Normal=0.5
    Restitution=&UniformRangedFloat(Minimum=0.2, Maximum=0.6)
   )
   &SolveForcesAndVelocity(Clamp Velocity=false, Limit Acceleration=false)
   &ScaleSpriteSize(Scale Factor=&Vector2DFromFloat(Value=&FloatFromCurve(FloatCurve=DataInterface() {propertyValues {Curve {keys [<at 0.7 0.9701>,<at 0.9011 0.9191>,<at 1.0011 0.002>]}}})))
   &ScaleSpriteSizeBySpeed(Max Scale Factor=<1 1.5>, Min Scale Factor=<0.7 0.7>, Velocity Threshold=250) {enabled=false}
   &Color(
    Color=<rgba 0.2917 0.2917 0.2917 1>
    Scale Alpha=&FloatFromCurve(FloatCurve=DataInterface() {propertyValues {Curve {keys [
         {arriveTangent=3.5481, interpMode=RCIM_Cubic, leaveTangent=3.5481, tangentMode=RCTM_User, time=0, value=1}
         {arriveTangent=-1, interpMode=RCIM_Cubic, leaveTangent=-1, time=0.2243, value=1}
         {arriveTangent=-0.9169, interpMode=RCIM_Cubic, leaveTangent=-0.9169, tangentMode=RCTM_User, time=1, value=0}
        ]}}})
   )
  ]
  rendererClasses [&NiagaraSpriteRendererProperties]
  renderers [{rendererClass=&NiagaraSpriteRendererProperties}]
 }
 Right_Leg_Dust_Kickoff {
  emitterUpdateScript [
   &EmitterState(
    Life Cycle Mode=::NewEnumerator1
    Loop Behavior=::NewEnumerator2
    Loop Count=2
    Loop Delay=0.3
    Loop Duration=1.15
    Recalculate Duration Each Loop=true
    UseLoopDelay=false
   )
   &SpawnBurst_Instantaneous(Spawn Count=10, Use Spawn Probability=false)
   &SpawnRate(
    SpawnGroup=1
    SpawnRate=&FloatFromCurve(
     FloatCurve=DataInterface() {propertyValues {Curve {keys [<at 0 1 RCIM_Cubic>,<at 0.5 0 RCIM_Cubic>]}}}
     CurveIndex=$Emitter.NormalizedLoopAge
     Scale Curve=15
    )
    Use Spawn Probability=false
   )
  ]
  particleSpawnScript [
   /Niagara/Modules/Spawn/Initialization/V2/InitializeParticle(
    Mass Mode=::NewEnumerator1
    Sprite Rotation Angle=&RandomRangeFloat(Minimum=0, Maximum=360)
    Sprite Rotation Mode=::NewEnumerator0
    Sprite Size Mode=::NewEnumerator3
    Uniform Sprite Size=50
    UsePositionOffset=false
    Write Material Random=true
   )
   &SkeletalMeshLocation(
    Bone Sampling Mode=::NewEnumerator4
    Inherit Velocity=false
    MaskSpawnGroup=false
    Sampled Position Offset=&MakeVector(
     X=-15
     Y=&MakeCustomFloatFromBool(
      Bool=HlslExpression() {hlslExpression="Emitter.SpawnGroup == 1"}
      True Float=&RandomRangeFloat(Minimum=-36, Maximum=-50)
      False Float=-30
     )
     Z=10
    )
    Sampled Position Offset Coordinate Space:ENiagaraMeshTransforms=::NewEnumerator3
    Skeletal Mesh=$User.SkeletalMesh
   )
   &SphereLocation(
    MaskSpawnGroup=false
    Override Seed=false
    Sphere Radius=6
    Surface Only=false
    UseNonUniformScale=true
    UseOffset=false
    UseSphereOrientation=false
   )
  ]
  particleUpdateScript [
   &ParticleState()
   &Color(
    Color=&ColorFromCurve(ColorCurve=DataInterface() {propertyValues {
       AlphaCurve {keys [<at 0 1 RCIM_Cubic>,<at 1 1 RCIM_Cubic>]}
       BlueCurve {keys [<at 0 0.4799 RCIM_Cubic>,<at 1 0.4799 RCIM_Cubic>]}
       GreenCurve {keys [<at 0 0.7637 RCIM_Cubic>,<at 1 0.7637 RCIM_Cubic>]}
       RedCurve {keys [<at 0 1 RCIM_Cubic>,<at 1 1 RCIM_Cubic>]}
      }})
    Scale Alpha=&FloatFromCurve(
     FloatCurve=DataInterface() {propertyValues {Curve {keys [
         {interpMode=RCIM_Cubic, leaveTangent=11.125, tangentMode=RCTM_Break, time=0, value=0}
         {interpMode=RCIM_Cubic, leaveTangent=-2.2426, tangentMode=RCTM_Break, time=0.25, value=1.2}
         {interpMode=RCIM_Cubic, tangentMode=RCTM_Break, time=1, value=0}
        ]}}}
     CurveIndex=&MakeCustomFloatFromBool(Bool=HlslExpression() {hlslExpression="Emitter.SpawnGroup == 1"}, True Float=$Emitter.LoopedAge, False Float=$Particles.NormalizedAge)
     Scale Curve=0.2
    )
   )
   &Drag(
    Drag=&FloatFromCurve(FloatCurve=DataInterface() {propertyValues {Curve {keys [
         {interpMode=RCIM_Cubic, tangentMode=RCTM_Break, time=0, value=1}
         {arriveTangent=-358.3112, interpMode=RCIM_Cubic, tangentMode=RCTM_Break, time=0.3, value=500}
        ]}}})
    Use Linear Drag=true
    Use Rotational Drag=false
   )
   &GravityForce(Gravity=<0 0 -2500>)
   &DynamicMaterialParameters(
    Index 0 Param 1=&Add_Float(
     A=$Particles.RandomFloat
     B=&FloatFromCurve(
      FloatCurve=DataInterface() {propertyValues {Curve {keys [{arriveTangent=2.7533, time=0, value=1}, {arriveTangent=0.0529, leaveTangent=0.2021, time=1, value=1.5}]}}}
      CurveIndex=&Add_Float(A=&Multiply_Float(A=$Particles.Initial.MaterialRandom, B=0.5), B=$Particles.NormalizedAge)
     )
    )
    Index 0 Param 2=&FloatFromCurve(FloatCurve=DataInterface() {propertyValues {Curve {keys [
         {arriveTangent=0.0453, interpMode=RCIM_Cubic, tangentMode=RCTM_Break, time=0, value=-1}
         {arriveTangent=0.0604, interpMode=RCIM_Cubic, leaveTangent=0.0604, tangentMode=RCTM_User, time=1, value=-1}
        ]}}})
    Index 0 Param 3=&FloatFromCurve(FloatCurve=DataInterface() {propertyValues {Curve {keys [
         {arriveTangent=-4.113, interpMode=RCIM_Cubic, leaveTangent=-2.1889, tangentMode=RCTM_User, time=0, value=1}
         {arriveTangent=0.2347, interpMode=RCIM_Cubic, leaveTangent=0.2723, tangentMode=RCTM_Break, time=1, value=0}
        ]}}})
    Index 0 Param 4=&FloatFromCurve(FloatCurve=DataInterface() {propertyValues {Curve {keys [<at 0 1>,<at 1 0>]}}})
    Index 1 Param 1=&FloatFromCurve(
     FloatCurve=DataInterface() {propertyValues {Curve {keys [
         {interpMode=RCIM_Cubic, leaveTangent=4.7088, tangentMode=RCTM_Break, time=0, value=0}
         {arriveTangent=4.2257, interpMode=RCIM_Cubic, leaveTangent=4.2257, tangentMode=RCTM_User, time=0.25, value=1}
        ]}}}
     Scale Curve=&RandomRangeFloat(Minimum=1.5, Maximum=1.5)
    )
    Index 1 Param 2=&FloatFromCurve(
     FloatCurve=DataInterface() {propertyValues {Curve {keys [<at 0 0>, {leaveTangent=0.9066, time=1, value=1}]}}}
     Scale Curve=&RandomRangeFloat(Minimum=0.5, Maximum=0.6)
    )
    Index 1 Param 3=0.25
    Index 1 Param 4=&FloatFromCurve(
     FloatCurve=DataInterface() {propertyValues {Curve {keys [{arriveTangent=-0.3224, interpMode=RCIM_Cubic, leaveTangent=-0.3224, tangentMode=RCTM_User, time=0, value=1}, <at 1 0.9>]}}}
     Scale Curve=&RandomRangeFloat(Minimum=1.1, Maximum=1.4)
    )
    Index 2 Param 1=&FloatFromCurve(FloatCurve=DataInterface() {propertyValues {Curve {keys [{interpMode=RCIM_Cubic, leaveTangent=1.4673, tangentMode=RCTM_Break, time=0, value=0}, <at 1 1>]}}})
    Index 2 Param 2=&FloatFromCurve(
     FloatCurve=DataInterface() {propertyValues {Curve {keys [
         {arriveTangent=1.9056, interpMode=RCIM_Cubic, leaveTangent=1.9056, tangentMode=RCTM_User, time=0, value=0.2}
         {arriveTangent=0.0259, interpMode=RCIM_Cubic, leaveTangent=0.0259, tangentMode=RCTM_User, time=1, value=1.4}
        ]}}}
     Scale Curve=&RandomRangeFloat(Minimum=0.1, Maximum=0.4)
    )
    Index 2 Param 3=&FloatFromCurve(
     FloatCurve=DataInterface() {propertyValues {Curve {keys [
         {interpMode=RCIM_Cubic, leaveTangent=2.6052, tangentMode=RCTM_Break, time=0, value=1}
         {interpMode=RCIM_Cubic, leaveTangent=11.1233, tangentMode=RCTM_Break, time=1, value=2}
        ]}}}
     Scale Curve=2
    )
    Index 2 Param 4=&FloatFromCurve(FloatCurve=DataInterface() {propertyValues {Curve {keys [{interpMode=RCIM_Cubic, tangentMode=RCTM_Break, time=0, value=1}, {interpMode=RCIM_Cubic, tangentMode=RCTM_Break, time=1, value=1.1}]}}})
    Module.Param0WriteEnabled=true
    Module.Param0WriteEnabled1=false
    Module.Param0WriteEnabled2=true
    Module.Param1WriteEnabled=false
    Module.Param1WriteEnabled1=false
    Module.Param1WriteEnabled2=true
    Module.Param2WriteEnabled=true
    Module.Param2WriteEnabled1=true
    Module.Param2WriteEnabled2=true
    Module.Param3WriteEnabled=true
    Module.Param3WriteEnabled1=true
    Module.Param3WriteEnabled2=true
    Write Parameter Index 1=true
    Write Parameter Index 2=true
   )
   &ScaleSpriteSize(Scale Factor=&Vector2DFromFloat(Value=&FloatFromCurve(FloatCurve=DataInterface() {propertyValues {Curve {keys [
          {interpMode=RCIM_Cubic, leaveTangent=-0.1211, tangentMode=RCTM_Break, time=0, value=0.42}
          {arriveTangent=-0.1422, interpMode=RCIM_Cubic, tangentMode=RCTM_Break, time=0.5, value=1}
         ]}}})))
   &AddVelocity(Velocity=&MakeVector(
     X=0
     Y=-250
     Z=&FloatFromCurve(FloatCurve=DataInterface() {propertyValues {Curve {keys [
          {interpMode=RCIM_Cubic, leaveTangent=-7.0775, tangentMode=RCTM_Break, time=0, value=1}
          {arriveTangent=-5.0843, interpMode=RCIM_Cubic, tangentMode=RCTM_Break, time=1, value=100}
         ]}}})
    ))
   &SolveForcesAndVelocity(Clamp Velocity=false, Limit Acceleration=false)
  ]
  rendererClasses [&NiagaraSpriteRendererProperties]
  renderers [{rendererClass=&NiagaraSpriteRendererProperties}]
 }
 Right_Leg_Rock_Kickoff {
  simTarget=GPUComputeSim
  emitterUpdateScript [
   &EmitterScalability(
    Module.bApplySpawnCountScale=false
    Module.bApplySpawnCountScaleByDistanceFraction=false
    Module.bUseMaxDistance=false
    Module.bUseMinDistance=false
   ) {enabled=false}
   &EmitterLifeCycle(NextLoopDuration=4.4208) {enabled=false}
   &EmitterState(
    Life Cycle Mode=::NewEnumerator1
    Loop Behavior=::NewEnumerator2
    Loop Count=2
    Loop Duration=1.15
    Recalculate Duration Each Loop=true
    UseLoopDelay=false
   )
   &SpawnBurst_Instantaneous(Spawn Count=&RandomRangeInt(Minimum=6, Maximum=10), Spawn Time=0.1, Use Spawn Probability=false)
  ]
  particleSpawnScript [
   /Niagara/Modules/Spawn/Initialization/InitializeParticle(
    Lifetime=&UniformRangedFloat(Minimum=0.5, Maximum=0.7)
    Mass=&RandomRangeFloat(Minimum=0.5, Maximum=2.3)
    Module.Write Color=true
    Module.Write Lifetime=true
    Module.Write Mass=true
    Module.Write Material Random=true
    Module.Write Position=false
    Module.Write Scale=false
    Module.Write SpriteRotation=false
    Module.Write SpriteSize=true
    Position=&SimulationPosition()
    Sprite Size=&Vector2DFromFloat(Value=&FloatFromCurve(
      FloatCurve=DataInterface() {propertyValues {Curve {keys [<at 0.1 1>,<at 0.2 0.5>,<at 1.0011 0.5>]}}}
      Scale Curve=&RandomRangeFloat(Minimum=1, Maximum=6)
     ))
   )
   &SkeletalMeshLocation(
    Bone Sampling Mode=::NewEnumerator4
    Inherit Velocity=false
    MaskSpawnGroup=false
    Sampled Position Offset=&MakeVector(
     X=-15
     Y=&MakeCustomFloatFromBool(
      Bool=HlslExpression() {hlslExpression="Emitter.SpawnGroup == 1"}
      True Float=&RandomRangeFloat(Minimum=-36, Maximum=-50)
      False Float=-30
     )
     Z=10
    )
    Sampled Position Offset Coordinate Space:ENiagaraMeshTransforms=::NewEnumerator3
    Skeletal Mesh=$User.SkeletalMesh
   )
   &SphereLocation(
    MaskSpawnGroup=false
    Override Seed=false
    Sphere Radius=10
    Surface Only=false
    UseNonUniformScale=false
    UseOffset=false
    UseSphereOrientation=false
   )
   &AddVelocity(Velocity=&MakeVector(
     X=0
     Y=&RandomRangeFloat(Minimum=-10, Maximum=-300)
     Z=&RandomRangeFloat(Minimum=40, Maximum=200)
    ))
  ]
  particleUpdateScript [
   &ParticleState()
   &Drag(Drag=&RandomRangeFloat(Minimum=3, Maximum=5), Use Linear Drag=true, Use Rotational Drag=false)
   &GravityForce(Gravity=<0 0 -900>)
   &Collision(
    Advanced Aging Rate=1
    Age Colliding Particles=false
    Friction=0.8
    Friction During a Bounce=0.1
    Kill Offscreen Particles=true
    Method for Calculating Particle Radius:ENiagaraCalculateRadiusOptions=::NewEnumerator1
    Module.Control Rest State Via Surface Normal=false
    Module.EnableMaxCollisionCount=false
    Module.Randomize Collision Normal Vector=true
    Particle Radius Scale=0.1
    Randomize Collision Normal=0.5
    Restitution=&UniformRangedFloat(Minimum=0.2, Maximum=0.6)
   )
   &SolveForcesAndVelocity(Clamp Velocity=false, Limit Acceleration=false)
   &ScaleSpriteSize(Scale Factor=&Vector2DFromFloat(Value=&FloatFromCurve(FloatCurve=DataInterface() {propertyValues {Curve {keys [<at 0.7 0.9701>,<at 0.9011 0.9191>,<at 1.0011 0.002>]}}})))
   &ScaleSpriteSizeBySpeed(Max Scale Factor=<1 1.5>, Min Scale Factor=<0.7 0.7>, Velocity Threshold=250) {enabled=false}
   &Color(
    Color=<rgba 0.2917 0.2917 0.2917 1>
    Scale Alpha=&FloatFromCurve(FloatCurve=DataInterface() {propertyValues {Curve {keys [
         {arriveTangent=3.5481, interpMode=RCIM_Cubic, leaveTangent=3.5481, tangentMode=RCTM_User, time=0, value=1}
         {arriveTangent=-1, interpMode=RCIM_Cubic, leaveTangent=-1, time=0.2243, value=1}
         {arriveTangent=-0.9169, interpMode=RCIM_Cubic, leaveTangent=-0.9169, tangentMode=RCTM_User, time=1, value=0}
        ]}}})
   )
  ]
  rendererClasses [&NiagaraSpriteRendererProperties]
  renderers [{rendererClass=&NiagaraSpriteRendererProperties}]
 }
}
summary {
 systemName=NS_GetTheHorns
 emitters [
  {emitterName=Right_Leg_Dust_Kickoff, rendererClasses [&NiagaraSpriteRendererProperties]}
  {emitterName=Left_Leg_Dust_Kickoff, rendererClasses [&NiagaraSpriteRendererProperties]}
  {emitterName=Initial_Left_Leg_Dust_Kickoff, rendererClasses [&NiagaraSpriteRendererProperties]}
  {emitterName=Right_Leg_Rock_Kickoff, simTarget=GPUComputeSim, rendererClasses [&NiagaraSpriteRendererProperties]}
  {emitterName=Left_Leg_Rock_Kickoff, simTarget=GPUComputeSim, rendererClasses [&NiagaraSpriteRendererProperties]}
  {emitterName=Initial_Left_Leg_Rock_Kickoff, simTarget=GPUComputeSim, rendererClasses [&NiagaraSpriteRendererProperties]}
 ]
 userVariables [User.SkeletalMesh {
   type=NiagaraDataInterfaceSkeletalMesh
   defaultValue {
    struct=NiagaraExt_VariableValue_DataInterface
    value {dataInterface=&self:NiagaraDataInterfaceSkeletalMesh_0}
   }
  }]
}
